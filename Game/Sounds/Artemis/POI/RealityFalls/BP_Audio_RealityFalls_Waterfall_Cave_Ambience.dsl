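; Audio Blueprint that shifts waterfall and cave ambience loops for the Reality Falls POI as the local pawn moves between zones.
; The interior/exterior box sets AreYouInside and crossfades loops, though exiting leaves the PrefabB interior loop, faded out at BeginPlay, playing.
; The lower-falls box silences the top exterior loop until exit; the upper-falls box fades the bottom exterior loop to half, and its exit FadeIn also targets half volume.
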(event EventBeginPlay
  ;[1 1 1] Beging Play - Fade Out Cave Interiors
  (Audio|Components|Audio|FadeOut (Variables|Default|GetAudio_Waterfall_PrefabB_Bottom_Interior_Loop) (Variables|Default|GetFadeTime) 0.0 "Linear" :!class /Script/Engine.AudioComponent))

(event OnComponentBeginOverlap(Box_Collision_Int_Ext)
  (params (OverlappedComponent :type Primitive%20Component%20Object%20Reference) (OtherActor :type Actor%20Object%20Reference) (OtherComp :type Primitive%20Component%20Object%20Reference) (OtherBodyIndex :type Integer) (bFromSweep :type Boolean) (SweepResult :type Hit%20Result%20Structure%20%28by%20ref%29))
  ;[1 1 1] Waterfall Loops Interior/Exterior Collision
  (if (HelperFunctions|IsLocalPawn OtherActor)
    (Variables|Default|SetAreYouInside true)
    (Audio|Components|Audio|FadeOut (Variables|Default|GetAudio_Waterfall_PrefabA_Bottom_Ext_Loop) 5.0 0.0 "Sin" :!class /Script/Engine.AudioComponent)
    (Audio|Components|Audio|FadeIn (Variables|Default|GetAudio_Waterfall_PrefabA_Bottom_Interior_Loop) 5.0 1.0 0.0 "Sin" :!class /Script/Engine.AudioComponent)
    (Audio|Components|Audio|FadeIn (Variables|Default|GetAudio_Waterfall_PrefabB_Bottom_Interior_Loop) 2.0 1.0 0.0 "Sin" :!class /Script/Engine.AudioComponent)
    (Audio|Components|Audio|StopDelayed (Variables|Default|GetAudio_Waterfall_PrefabA_Bottom_Ext_Loop) 5.0)))

(event Collision|EventActorBeginOverlap
  (params (OtherActor :type Actor%20Object%20Reference)))

(event EventTick
  (params (DeltaSeconds :type Float%20%28single-precision%29)))

(event OnComponentEndOverlap(Box_Collision_Int_Ext)
  (params (OverlappedComponent :type Primitive%20Component%20Object%20Reference) (OtherActor :type Actor%20Object%20Reference) (OtherComp :type Primitive%20Component%20Object%20Reference) (OtherBodyIndex :type Integer))
  (bind _audio_waterfall_prefaba_bottom_interior_loop (Variables|Default|GetAudio_Waterfall_PrefabA_Bottom_Interior_Loop))
  ;[1 1 1] Waterfall Loops Interior/Exterior Collision
  (if (HelperFunctions|IsLocalPawn OtherActor)
    (Variables|Default|SetAreYouInside false)
    (Audio|Components|Audio|FadeIn (Variables|Default|GetAudio_Waterfall_PrefabA_Bottom_Ext_Loop) 5.0 1.0 0.0 "Sin" :!class /Script/Engine.AudioComponent)
    (Audio|Components|Audio|FadeOut _audio_waterfall_prefaba_bottom_interior_loop 5.0 0.0 "Sin" :!class /Script/Engine.AudioComponent)
    (Audio|Components|Audio|StopDelayed _audio_waterfall_prefaba_bottom_interior_loop 5.0)))

(event OnComponentBeginOverlap(Box_Collision_LowerFalls_Top_Ext_Fades)
  (params (OverlappedComponent :type Primitive%20Component%20Object%20Reference) (OtherActor :type Actor%20Object%20Reference) (OtherComp :type Primitive%20Component%20Object%20Reference) (OtherBodyIndex :type Integer) (bFromSweep :type Boolean) (SweepResult :type Hit%20Result%20Structure%20%28by%20ref%29))
  (bind _audio_waterfall_prefabb_top_ext_loop (Variables|Default|GetAudio_Waterfall_PrefabB_Top_Ext_Loop))
  ;[1 1 1] When in Lower Falls Fade Out Top Waterfall Cues
  (if (HelperFunctions|IsLocalPawn OtherActor)
    (Audio|Components|Audio|FadeOut _audio_waterfall_prefabb_top_ext_loop 5.0 0.0 "Linear" :!class /Script/Engine.AudioComponent)
    (Audio|Components|Audio|StopDelayed _audio_waterfall_prefabb_top_ext_loop 5.1)))

(event OnComponentEndOverlap(Box_Collision_LowerFalls_Top_Ext_Fades)
  (params (OverlappedComponent :type Primitive%20Component%20Object%20Reference) (OtherActor :type Actor%20Object%20Reference) (OtherComp :type Primitive%20Component%20Object%20Reference) (OtherBodyIndex :type Integer))
  (bind _audio_waterfall_prefabb_top_ext_loop (Variables|Default|GetAudio_Waterfall_PrefabB_Top_Ext_Loop))
  ;[1 1 1] When in Lower Falls Fade Out Top Waterfall Cues
  (if (HelperFunctions|IsLocalPawn OtherActor)
    (Audio|Components|Audio|Play _audio_waterfall_prefabb_top_ext_loop (Math|Random|RandomFloatinRange 1.0 2.0))
    (Audio|Components|Audio|FadeIn _audio_waterfall_prefabb_top_ext_loop 5.0 1.0 0.0 "Linear" :!class /Script/Engine.AudioComponent)))

(event OnComponentBeginOverlap(Box_Collision_UpperFalls_Bottom_Ext_Fades)
  (params (OverlappedComponent :type Primitive%20Component%20Object%20Reference) (OtherActor :type Actor%20Object%20Reference) (OtherComp :type Primitive%20Component%20Object%20Reference) (OtherBodyIndex :type Integer) (bFromSweep :type Boolean) (SweepResult :type Hit%20Result%20Structure%20%28by%20ref%29))
  ;[1 1 1] When on top of the Falls Control The Volume of The Bottom Exterior Cues
  (if (HelperFunctions|IsLocalPawn OtherActor)
    (Audio|Components|Audio|FadeOut (Variables|Default|GetAudio_Waterfall_PrefabA_Bottom_Ext_Loop) 5.0 0.5 "Linear" :!class /Script/Engine.AudioComponent) ; Fade out Bottom Cues by Half
  ))

(event OnComponentEndOverlap(Box_Collision_UpperFalls_Bottom_Ext_Fades)
  (params (OverlappedComponent :type Primitive%20Component%20Object%20Reference) (OtherActor :type Actor%20Object%20Reference) (OtherComp :type Primitive%20Component%20Object%20Reference) (OtherBodyIndex :type Integer))
  ;[1 1 1] When on top of the Falls Control The Volume of The Bottom Exterior Cues
  (if (HelperFunctions|IsLocalPawn OtherActor)
    (Audio|Components|Audio|FadeIn (Variables|Default|GetAudio_Waterfall_PrefabA_Bottom_Ext_Loop) 5.0 0.5 1.0 "Linear" :!class /Script/Engine.AudioComponent) ; Fade In Bottom Exterior Cues to Original Vol
  ))

(fn ConstructionScript)
